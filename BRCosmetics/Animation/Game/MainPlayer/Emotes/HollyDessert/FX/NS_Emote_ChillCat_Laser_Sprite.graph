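# Niagara effect for the ChillCat emote in the HollyDessert folder, showing a volley of red lasers with impact glows, dust and smoke.
# Timed bursts across thirteen spawn groups map to hardcoded impact points through nested MaskVectorBySpawnGroup inputs.
# Laser sprites spawn in a box and are pulled to those points by attraction, stretching with speed.
# Glow sprites with lights and GPU dust follow shortly after, while Reload_Smoke spawning is gated by System.Spawn.
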NiagaraSystem /BRCosmetics/Animation/Game/MainPlayer/Emotes/HollyDessert/FX/NS_Emote_ChillCat_Laser_Sprite
dependencies {
 usedDataInterfaces [&NiagaraDataInterfaceCurve, &NiagaraDataInterfaceVector2DCurve, &NiagaraDataInterfaceColorCurve]
 usedDynamicInputs [&RandomBool]
 usedModules [
  &self:SystemSpawnScript.SystemScriptSource.SystemScriptGraph.NiagaraNodeAssignment_1.SetVariables_#1
  &SystemState
 ]
}
emitters {
 Dust {
  simTarget=GPUComputeSim
  emitterUpdateScript [
   &EmitterState(
    Enable Distance Culling=true
    Enable Visibility Culling=true
    Life Cycle Mode=::NewEnumerator1
    Loop Behavior=::NewEnumerator1
    Loop Delay=&RandomRangeFloat(Minimum=0.1, Maximum=0.125)
    Loop Duration=0.65
    MaxDistance=4000
    Module.bApplySpawnCountScaleByDistance=true
    Module.bUseMaxDistance=true
    Module.bUseMinDistance=false
    Scalability Mode=::NewEnumerator1
    Scale Spawn Count=true
    Spawn Count Scale By Distance=DataInterface() {propertyValues {Curve {keys [<at 0 1>,<at 0 1>,<at 1 0>]}}}
    UseLoopDelay=true
   )
   &SpawnBurst_Instantaneous(Spawn Count=5, Spawn Time=0.12, Use Spawn Probability=false, UseLoopCountLimit=false)
   &SpawnBurst_Instantaneous(Spawn Count=5, Spawn Group=1, Spawn Time=0.07, Use Spawn Probability=false, UseLoopCountLimit=false) {module=SpawnBurst_Instantaneous001}
   &SpawnBurst_Instantaneous(Spawn Count=5, Spawn Group=2, Spawn Time=0.17, Use Spawn Probability=false, UseLoopCountLimit=false) {module=SpawnBurst_Instantaneous002}
   &SpawnBurst_Instantaneous(Spawn Group=3, Spawn Time=0.28, Use Spawn Probability=false, UseLoopCountLimit=true) {module=SpawnBurst_Instantaneous003}
   &SpawnBurst_Instantaneous(Spawn Count=5, Spawn Group=4, Spawn Time=0.3, Use Spawn Probability=false, UseLoopCountLimit=false) {module=SpawnBurst_Instantaneous004}
   &SpawnBurst_Instantaneous(Spawn Group=5, Spawn Time=0.38, Use Spawn Probability=false, UseLoopCountLimit=true) {module=SpawnBurst_Instantaneous005}
   &SpawnBurst_Instantaneous(Spawn Group=6, Spawn Time=0.31, Use Spawn Probability=false, UseLoopCountLimit=true) {module=SpawnBurst_Instantaneous006}
   &SpawnBurst_Instantaneous(Spawn Group=7, Spawn Time=0.07, Use Spawn Probability=false, UseLoopCountLimit=true) {module=SpawnBurst_Instantaneous007}
   &SpawnBurst_Instantaneous(Spawn Group=8, Spawn Time=0.07, Use Spawn Probability=false, UseLoopCountLimit=true) {module=SpawnBurst_Instantaneous008}
   &SpawnBurst_Instantaneous(Spawn Group=9, Spawn Time=0.16, Use Spawn Probability=false, UseLoopCountLimit=true) {module=SpawnBurst_Instantaneous009}
   &SpawnBurst_Instantaneous(Spawn Group=10, Spawn Time=0.24, Use Spawn Probability=false, UseLoopCountLimit=true) {module=SpawnBurst_Instantaneous010}
   &SpawnBurst_Instantaneous(Spawn Group=11, Spawn Time=0.29, Use Spawn Probability=false, UseLoopCountLimit=true) {module=SpawnBurst_Instantaneous011}
   &SpawnBurst_Instantaneous(Spawn Group=12, Spawn Time=0.26, Use Spawn Probability=false, UseLoopCountLimit=true) {module=SpawnBurst_Instantaneous012}
  ]
  particleSpawnScript [
   &InitializeParticle(
    Color Maximum=<rgba 0.2656 0.2656 0.2656 1>
    Color Minimum=<rgba 0.3627 0.3684 0.3854 1>
    Color Mode=::NewEnumerator2
    Lifetime Max=1
    Lifetime Min=0.6
    Lifetime Mode=::NewEnumerator1
    Position Offset=<-6 0 0>
    Sprite Rotation Mode=::NewEnumerator2
    Sprite Size Mode=::NewEnumerator1
    Uniform Sprite Size Max=8
    Uniform Sprite Size Min=3
    UsePositionOffset=true
    Write Material Random=true
   )
   /SetVariables(Particles.Position=&MaskVectorBySpawnGroup(
     Masked Vector=<-6 0 0>
     Passthrough Vector=&MaskVectorBySpawnGroup(
      Masked Vector=&AddVectorToPosition(Position=<0 21 -2>, Vector=<-6 0 0>)
      Passthrough Vector=&MaskVectorBySpawnGroup(
       Masked Vector=&AddVectorToPosition(Position=<0 40 -3>, Vector=<-6 0 0>)
       Passthrough Vector=&MaskVectorBySpawnGroup(
        Masked Vector=&AddVectorToPosition(Position=<0 34 -10>, Vector=<-6 0 0>)
        Passthrough Vector=&MaskVectorBySpawnGroup(
         Masked Vector=&AddVectorToPosition(Position=<0 54 -7>, Vector=<-6 0 0>)
         Passthrough Vector=&MaskVectorBySpawnGroup(
          Masked Vector=&AddVectorToPosition(Position=<0 12 -8>, Vector=<-6 0 0>)
          Passthrough Vector=&MaskVectorBySpawnGroup(
           Masked Vector=&AddVectorToPosition(Position=<0 -2 -6>, Vector=<-6 0 0>)
           Passthrough Vector=&MaskVectorBySpawnGroup(
            Masked Vector=&AddVectorToPosition(Position=<0 19 -12>, Vector=<-6 0 0>)
            Passthrough Vector=&MaskVectorBySpawnGroup(
             Masked Vector=&AddVectorToPosition(Position=<0 45 -12>, Vector=<-6 0 0>)
             Passthrough Vector=&MaskVectorBySpawnGroup(
              Masked Vector=&AddVectorToPosition(Position=<0 36 -15>, Vector=<-6 0 0>)
              Passthrough Vector=&MaskVectorBySpawnGroup(
               Masked Vector=&AddVectorToPosition(Position=<0 41 -15>, Vector=<-6 0 0>)
               Passthrough Vector=&MaskVectorBySpawnGroup(
                Masked Vector=&AddVectorToPosition(Position=<0 6 -14>, Vector=<-6 0 0>)
                Passthrough Vector=&AddVectorToPosition(Position=<0 -6 -10>, Vector=<-6 0 0>)
                Spawn Group Mask=11
               )
               Spawn Group Mask=10
              )
              Spawn Group Mask=9
             )
             Spawn Group Mask=8
            )
            Spawn Group Mask=7
           )
           Spawn Group Mask=6
          )
          Spawn Group Mask=5
         )
         Spawn Group Mask=4
        )
        Spawn Group Mask=3
       )
       Spawn Group Mask=2
      )
      Spawn Group Mask=1
     )
    )) {isSetParameters=true}
   /SetVariables(Particles.InitSpriteRotRate=&RandomRangeFloat(Minimum=-3, Maximum=3)) {isSetParameters=true}
   &SphereLocation(
    MaskSpawnGroup=false
    Override Seed=false
    Sphere Origin:NiagaraPosition=&MaskVectorBySpawnGroup(
     Masked Vector=<-6 0 0>
     Passthrough Vector=&MaskVectorBySpawnGroup(
      Masked Vector=&AddVectorToPosition(Position=<0 21 -2>, Vector=<-6 0 0>)
      Passthrough Vector=&MaskVectorBySpawnGroup(
       Masked Vector=&AddVectorToPosition(Position=<0 40 -3>, Vector=<-6 0 0>)
       Passthrough Vector=&MaskVectorBySpawnGroup(
        Masked Vector=&AddVectorToPosition(Position=<0 34 -10>, Vector=<-6 0 0>)
        Passthrough Vector=&MaskVectorBySpawnGroup(
         Masked Vector=&AddVectorToPosition(Position=<0 54 -7>, Vector=<-6 0 0>)
         Passthrough Vector=&MaskVectorBySpawnGroup(
          Masked Vector=&AddVectorToPosition(Position=<0 12 -8>, Vector=<-6 0 0>)
          Passthrough Vector=&MaskVectorBySpawnGroup(
           Masked Vector=&AddVectorToPosition(Position=<0 -2 -6>, Vector=<-6 0 0>)
           Passthrough Vector=&MaskVectorBySpawnGroup(
            Masked Vector=&AddVectorToPosition(Position=<0 19 -12>, Vector=<-6 0 0>)
            Passthrough Vector=&MaskVectorBySpawnGroup(
             Masked Vector=&AddVectorToPosition(Position=<0 45 -12>, Vector=<-6 0 0>)
             Passthrough Vector=&MaskVectorBySpawnGroup(
              Masked Vector=&AddVectorToPosition(Position=<0 36 -15>, Vector=<-6 0 0>)
              Passthrough Vector=&MaskVectorBySpawnGroup(
               Masked Vector=&AddVectorToPosition(Position=<0 41 -15>, Vector=<-6 0 0>)
               Passthrough Vector=&MaskVectorBySpawnGroup(
                Masked Vector=&AddVectorToPosition(Position=<0 6 -14>, Vector=<-6 0 0>)
                Passthrough Vector=&AddVectorToPosition(Position=<0 -6 -10>, Vector=<-6 0 0>)
                Spawn Group Mask=11
               )
               Spawn Group Mask=10
              )
              Spawn Group Mask=9
             )
             Spawn Group Mask=8
            )
            Spawn Group Mask=7
           )
           Spawn Group Mask=6
          )
          Spawn Group Mask=5
         )
         Spawn Group Mask=4
        )
        Spawn Group Mask=3
       )
       Spawn Group Mask=2
      )
      Spawn Group Mask=1
     )
    )
    Sphere Radius=1
    Surface Only=false
    UseNonUniformScale=false
    UseOffset=false
    UseSphereOrientation=false
   )
   &AddVelocity(
    Module.Use Falloff Distance=false
    Override Seed=false
    Velocity Mode=::NewEnumerator1
    Velocity Origin=&MaskVectorBySpawnGroup(
     Masked Vector=<-6 0 0>
     Passthrough Vector=&MaskVectorBySpawnGroup(
      Masked Vector=&AddVectorToPosition(Position=<0 21 -2>, Vector=<-6 0 0>)
      Passthrough Vector=&MaskVectorBySpawnGroup(
       Masked Vector=&AddVectorToPosition(Position=<0 40 -3>, Vector=<-6 0 0>)
       Passthrough Vector=&MaskVectorBySpawnGroup(
        Masked Vector=&AddVectorToPosition(Position=<0 34 -10>, Vector=<-6 0 0>)
        Passthrough Vector=&MaskVectorBySpawnGroup(
         Masked Vector=&AddVectorToPosition(Position=<0 54 -7>, Vector=<-6 0 0>)
         Passthrough Vector=&MaskVectorBySpawnGroup(
          Masked Vector=&AddVectorToPosition(Position=<0 12 -8>, Vector=<-6 0 0>)
          Passthrough Vector=&MaskVectorBySpawnGroup(
           Masked Vector=&AddVectorToPosition(Position=<0 -2 -6>, Vector=<-6 0 0>)
           Passthrough Vector=&MaskVectorBySpawnGroup(
            Masked Vector=&AddVectorToPosition(Position=<0 19 -12>, Vector=<-6 0 0>)
            Passthrough Vector=&MaskVectorBySpawnGroup(
             Masked Vector=&AddVectorToPosition(Position=<0 45 -12>, Vector=<-6 0 0>)
             Passthrough Vector=&MaskVectorBySpawnGroup(
              Masked Vector=&AddVectorToPosition(Position=<0 36 -15>, Vector=<-6 0 0>)
              Passthrough Vector=&MaskVectorBySpawnGroup(
               Masked Vector=&AddVectorToPosition(Position=<0 41 -15>, Vector=<-6 0 0>)
               Passthrough Vector=&MaskVectorBySpawnGroup(
                Masked Vector=&AddVectorToPosition(Position=<0 6 -14>, Vector=<-6 0 0>)
                Passthrough Vector=&AddVectorToPosition(Position=<0 -6 -10>, Vector=<-6 0 0>)
                Spawn Group Mask=11
               )
               Spawn Group Mask=10
              )
              Spawn Group Mask=9
             )
             Spawn Group Mask=8
            )
            Spawn Group Mask=7
           )
           Spawn Group Mask=6
          )
          Spawn Group Mask=5
         )
         Spawn Group Mask=4
        )
        Spawn Group Mask=3
       )
       Spawn Group Mask=2
      )
      Spawn Group Mask=1
     )
    )
    Velocity Speed=&Lerp_Float(
     A=10
     B=20
     Alpha=&FloatFromCurve(
      FloatCurve=DataInterface() {propertyValues {Curve {keys [
          {interpMode=RCIM_Cubic, tangentMode=RCTM_User, time=0, value=0}
          {arriveTangent=0.8718, interpMode=RCIM_Cubic, leaveTangent=0.8718, tangentMode=RCTM_User, time=0.8, value=0.2}
          {arriveTangent=5.7883, interpMode=RCIM_Cubic, leaveTangent=5.7883, tangentMode=RCTM_User, time=1, value=1}
         ]}}}
      CurveIndex=&RandomRangeFloat(Minimum=0, Maximum=1)
     )
    )
   )
  ]
  particleUpdateScript [
   &ParticleState()
   &Drag(Drag=4, Use Linear Drag=true, Use Rotational Drag=true)
   &SpriteRotationRate(Rotation Rate=&Multiply_Float(
     A=$Particles.InitSpriteRotRate
     B=&FloatFromCurve(
      FloatCurve=DataInterface() {propertyValues {Curve {keys [
          {arriveTangent=-4.0107, interpMode=RCIM_Cubic, leaveTangent=-4.0107, tangentMode=RCTM_User, time=-0.0046, value=1.0113}
          {arriveTangent=-0.2496, interpMode=RCIM_Cubic, leaveTangent=-0.2496, tangentMode=RCTM_User, time=0.5, value=0.16}
         ]}}}
      CurveIndex=$Emitter.Age
      Scale Curve=150
     )
    ))
   &ScaleColor(
    Scale Alpha=&FloatFromCurve(
     FloatCurve=DataInterface() {propertyValues {Curve {keys [
         {interpMode=RCIM_Cubic, tangentMode=RCTM_User, time=0, value=0}
         {interpMode=RCIM_Cubic, tangentMode=RCTM_User, time=0.3, value=1}
         <at 0.7 0>
        ]}}}
     Scale Curve=&RandomRangeFloat(Minimum=0.7, Maximum=1)
    )
    Scale RGB=&VectorFromFloat(Value=&FloatFromCurve(FloatCurve=DataInterface() {propertyValues {Curve {keys [{arriveTangent=1.8145, interpMode=RCIM_Cubic, leaveTangent=4.5795, tangentMode=RCTM_Break, time=0, value=0.4}, <at 0.3387 0.9565>, <at 1 1>]}}}))
    ScaleA=true
    ScaleRGB=true
   )
   &ScaleSpriteSize(Scale Factor=&Vector2DFromFloat(Value=&FloatFromCurve(
      FloatCurve=DataInterface() {propertyValues {Curve {keys [
          {arriveTangent=2.0514, interpMode=RCIM_Cubic, leaveTangent=0.7961, tangentMode=RCTM_User, time=0, value=0.4}
          {arriveTangent=-0.0457, interpMode=RCIM_Cubic, leaveTangent=-0.0457, tangentMode=RCTM_User, time=1.0463, value=1}
         ]}}}
      Scale Curve=0.75
     )))
   &DynamicMaterialParameters(
    Index 0 Param 1=&Add_Float(
     A=$Particles.RandomFloat
     B=&FloatFromCurve(FloatCurve=DataInterface() {propertyValues {Curve {keys [
          {arriveTangent=2.7533, interpMode=RCIM_Cubic, leaveTangent=2.847, tangentMode=RCTM_Break, time=0, value=0}
          {arriveTangent=1.3581, interpMode=RCIM_Cubic, leaveTangent=1.3581, tangentMode=RCTM_User, time=0.2558, value=0.5125}
          {arriveTangent=0.0529, interpMode=RCIM_Cubic, leaveTangent=0.2021, tangentMode=RCTM_User, time=1, value=1}
         ]}}})
    )
    Index 0 Param 3=&FloatFromCurve(FloatCurve=DataInterface() {propertyValues {Curve {keys [{arriveTangent=-53.9199, time=0, value=0}, {leaveTangent=0.2723, time=0.25, value=0}]}}})
    Index 0 Param 4=&FloatFromCurve(FloatCurve=DataInterface() {propertyValues {Curve {keys [<at 0 1>,<at 1 0>]}}})
    Index 1 Param 1=&FloatFromCurve(FloatCurve=DataInterface() {propertyValues {Curve {keys [<at 0 0>, {arriveTangent=4.2257, interpMode=RCIM_Cubic, leaveTangent=4.2257, tangentMode=RCTM_User, time=0.25, value=0}]}}})
    Index 1 Param 2=&FloatFromCurve(FloatCurve=DataInterface() {propertyValues {Curve {keys [<at 0 0>]}}})
    Index 1 Param 3=&FloatFromCurve(FloatCurve=DataInterface() {propertyValues {Curve {keys [{interpMode=RCIM_Cubic, leaveTangent=6.3917, tangentMode=RCTM_Break, time=0, value=0.1}]}}})
    Index 1 Param 4=&FloatFromCurve(FloatCurve=DataInterface() {propertyValues {Curve {keys [{arriveTangent=-0.136, leaveTangent=-0.136, time=0, value=1}]}}})
    Index 2 Param 1=&FloatFromCurve(FloatCurve=DataInterface() {propertyValues {Curve {keys [
         {arriveTangent=1.2401, interpMode=RCIM_Cubic, leaveTangent=0.0353, tangentMode=RCTM_User, time=0, value=0}
         {arriveTangent=2.3594, interpMode=RCIM_Cubic, leaveTangent=2.2889, tangentMode=RCTM_User, time=1, value=1}
        ]}}})
    Index 2 Param 2=&FloatFromCurve(
     FloatCurve=DataInterface() {propertyValues {Curve {keys [
         {arriveTangent=1.9056, time=0, value=0.2}
         {arriveTangent=0.0259, interpMode=RCIM_Cubic, leaveTangent=0.0259, tangentMode=RCTM_User, time=1, value=1.4}
        ]}}}
     Scale Curve=&RandomRangeFloat(Minimum=0.5, Maximum=1)
    )
    Index 2 Param 3=&FloatFromCurve(FloatCurve=DataInterface() {propertyValues {Curve {keys [
         {interpMode=RCIM_Cubic, leaveTangent=2.3064, tangentMode=RCTM_Break, time=0, value=1}
         {interpMode=RCIM_Cubic, leaveTangent=11.1233, tangentMode=RCTM_Break, time=1, value=2}
        ]}}})
    Index 2 Param 4=&FloatFromCurve(
     FloatCurve=DataInterface() {propertyValues {Curve {keys [
         {arriveTangent=-0.6867, interpMode=RCIM_Cubic, leaveTangent=-0.6867, tangentMode=RCTM_User, time=0, value=1}
         {interpMode=RCIM_Cubic, leaveTangent=-1.2956, tangentMode=RCTM_Break, time=1, value=0.75}
        ]}}}
     Scale Curve=1.6
    )
    Module.Param0WriteEnabled=false
    Module.Param0WriteEnabled1=true
    Module.Param0WriteEnabled2=true
    Module.Param1WriteEnabled=false
    Module.Param1WriteEnabled1=true
    Module.Param1WriteEnabled2=false
    Module.Param2WriteEnabled=true
    Module.Param2WriteEnabled1=true
    Module.Param2WriteEnabled2=true
    Module.Param3WriteEnabled=false
    Module.Param3WriteEnabled1=true
    Module.Param3WriteEnabled2=true
    Write Parameter Index 1=true
    Write Parameter Index 2=true
   )
   &GravityForce(Gravity=&RandomRangeVector(Minimum=<0 0 10>, Maximum=<0 0 -130>))
   &SolveForcesAndVelocity(Clamp Velocity=false, Limit Acceleration=false)
   &ScaleColor(
    Scale Alpha=/Niagara/DynamicInputs/Math/LerpMultipleFloats(
     Float Count:ENiagara_MultipleLerpCount=::NewEnumerator1
     1:NiagaraFloat=1
     2:NiagaraFloat=1
     3:NiagaraFloat=0.75
     4:NiagaraFloat=1
     5:NiagaraFloat=?
     Alpha:NiagaraFloat=Linked() {linkedVariable {name=NPC.FortniteNPC.TimeOfDay_0-1, type=NiagaraFloat}}
    )
    ScaleA=true
    ScaleRGB=false
   ) {module=ScaleColor002}
  ]
  rendererClasses [&NiagaraSpriteRendererProperties]
  renderers [
   {rendererClass=&NiagaraSpriteRendererProperties}
   {rendererClass=&NiagaraSpriteRendererProperties, rendererIndex=1}
  ]
 }
 Glow {
  emitterUpdateScript [
   &EmitterState(
    Enable Distance Culling=true
    Enable Visibility Culling=true
    MaxDistance=7000
    Module.bUseMaxDistance=true
    Module.bUseMinDistance=false
    Scalability Mode=::NewEnumerator1
   )
   &SpawnBurst_Instantaneous(Spawn Time=0.13, Use Spawn Probability=false, UseLoopCountLimit=false)
   &SpawnBurst_Instantaneous(Spawn Group=1, Spawn Time=0.16, Use Spawn Probability=false, UseLoopCountLimit=false) {module=SpawnBurst_Instantaneous001}
   &SpawnBurst_Instantaneous(Spawn Group=2, Spawn Time=0.2, Use Spawn Probability=false, UseLoopCountLimit=false) {module=SpawnBurst_Instantaneous002}
   &SpawnBurst_Instantaneous(Spawn Group=3, Spawn Time=0.31, Use Spawn Probability=false, UseLoopCountLimit=true) {module=SpawnBurst_Instantaneous003}
   &SpawnBurst_Instantaneous(Spawn Group=4, Spawn Time=0.33, Use Spawn Probability=false, UseLoopCountLimit=true) {module=SpawnBurst_Instantaneous004}
   &SpawnBurst_Instantaneous(Spawn Group=5, Spawn Time=0.39, Use Spawn Probability=false, UseLoopCountLimit=true) {module=SpawnBurst_Instantaneous005}
   &SpawnBurst_Instantaneous(Spawn Group=6, Spawn Time=0.32, Use Spawn Probability=false, UseLoopCountLimit=true) {module=SpawnBurst_Instantaneous006}
   &SpawnBurst_Instantaneous(Spawn Group=7, Spawn Time=0.16, Use Spawn Probability=false, UseLoopCountLimit=true) {module=SpawnBurst_Instantaneous007}
   &SpawnBurst_Instantaneous(Spawn Group=8, Spawn Time=0.16, Use Spawn Probability=false, UseLoopCountLimit=true) {module=SpawnBurst_Instantaneous008}
   &SpawnBurst_Instantaneous(Spawn Group=9, Spawn Time=0.25, Use Spawn Probability=false, UseLoopCountLimit=true) {module=SpawnBurst_Instantaneous009}
   &SpawnBurst_Instantaneous(Spawn Group=10, Spawn Time=0.33, Use Spawn Probability=false, UseLoopCountLimit=true) {module=SpawnBurst_Instantaneous010}
   &SpawnBurst_Instantaneous(Spawn Group=11, Spawn Time=0.38, Use Spawn Probability=false, UseLoopCountLimit=true) {module=SpawnBurst_Instantaneous011}
   &SpawnBurst_Instantaneous(Spawn Group=12, Spawn Time=0.35, Use Spawn Probability=false, UseLoopCountLimit=true) {module=SpawnBurst_Instantaneous012}
  ]
  particleSpawnScript [
   &InitializeParticle(
    Color=<rgba 5 0 0 1>
    Lifetime=0.1
    Position Offset=<-6 0 0>
    Sprite Rotation Mode=::NewEnumerator2
    Sprite Size Mode=::NewEnumerator1
    Uniform Sprite Size Max=15
    Uniform Sprite Size Min=8
    UsePositionOffset=true
    Write Material Random=true
   )
   /SetVariables(Particles.Position=&MaskVectorBySpawnGroup(
     Masked Vector=<-6 0 0>
     Passthrough Vector=&MaskVectorBySpawnGroup(
      Masked Vector=&AddVectorToPosition(Position=<0 21 -2>, Vector=<-6 0 0>)
      Passthrough Vector=&MaskVectorBySpawnGroup(
       Masked Vector=&AddVectorToPosition(Position=<0 40 -3>, Vector=<-6 0 0>)
       Passthrough Vector=&MaskVectorBySpawnGroup(
        Masked Vector=&AddVectorToPosition(Position=<0 34 -10>, Vector=<-6 0 0>)
        Passthrough Vector=&MaskVectorBySpawnGroup(
         Masked Vector=&AddVectorToPosition(Position=<0 54 -7>, Vector=<-6 0 0>)
         Passthrough Vector=&MaskVectorBySpawnGroup(
          Masked Vector=&AddVectorToPosition(Position=<0 12 -8>, Vector=<-6 0 0>)
          Passthrough Vector=&MaskVectorBySpawnGroup(
           Masked Vector=&AddVectorToPosition(Position=<0 -2 -6>, Vector=<-6 0 0>)
           Passthrough Vector=&MaskVectorBySpawnGroup(
            Masked Vector=&AddVectorToPosition(Position=<0 19 -12>, Vector=<-6 0 0>)
            Passthrough Vector=&MaskVectorBySpawnGroup(
             Masked Vector=&AddVectorToPosition(Position=<0 45 -12>, Vector=<-6 0 0>)
             Passthrough Vector=&MaskVectorBySpawnGroup(
              Masked Vector=&AddVectorToPosition(Position=<0 36 -15>, Vector=<-6 0 0>)
              Passthrough Vector=&MaskVectorBySpawnGroup(
               Masked Vector=&AddVectorToPosition(Position=<0 41 -15>, Vector=<-6 0 0>)
               Passthrough Vector=&MaskVectorBySpawnGroup(
                Masked Vector=&AddVectorToPosition(Position=<0 6 -14>, Vector=<-6 0 0>)
                Passthrough Vector=&AddVectorToPosition(Position=<0 -6 -10>, Vector=<-6 0 0>)
                Spawn Group Mask=11
               )
               Spawn Group Mask=10
              )
              Spawn Group Mask=9
             )
             Spawn Group Mask=8
            )
            Spawn Group Mask=7
           )
           Spawn Group Mask=6
          )
          Spawn Group Mask=5
         )
         Spawn Group Mask=4
        )
        Spawn Group Mask=3
       )
       Spawn Group Mask=2
      )
      Spawn Group Mask=1
     )
    )) {isSetParameters=true}
   /SetVariables(Particles.SpriteRotation:NiagaraFloat=&RandomRangeFloat(Minimum=0, Maximum=1)) {isSetParameters=true}
  ]
  particleUpdateScript [
   &ParticleState()
   &ScaleSpriteSize(
    Uniform Curve Scale=0.5
    Uniform Curve Sprite Scale=DataInterface() {propertyValues {Curve {keys [
        {interpMode=RCIM_Cubic, tangentMode=RCTM_User, time=0, value=1}
        {interpMode=RCIM_Cubic, tangentMode=RCTM_User, time=0.3, value=2}
        {interpMode=RCIM_Cubic, tangentMode=RCTM_User, time=1, value=1}
       ]}}}
   )
   &ScaleColor(ScaleA=true, ScaleRGB=false)
   &DynamicMaterialParameters(
    Index 1 Param 1=15
    Module.Param0WriteEnabled1=true
    Module.Param1WriteEnabled1=true
    Module.Param2WriteEnabled1=true
    Module.Param3WriteEnabled1=true
    Write Parameter Index 0=false
    Write Parameter Index 1=true
   )
   &FN_Light_Attributes(
    Color=<rgba 1.5 0.0477 0 1>
    Create Color Attribute=true
    Create Exponent Attribute=true
    Create Position Attribute=true
    Create Radius Attribute=true
    Exponent=20
    Position=&Add_Vector(
     A=&ConvertPositionToVector(Input Position=$Particles.Position, Convert To Absolute Large World Vector=false)
     B=<0 0 0>
    )
    Radius=&Multiply_Float(A=150, B=1)
    Scale Color=&Multiply_Float(A=&FNTimeOfDayScalar(Night Scale=1, Morning Scale=1.5, Day Scale=2, Evening Scale=1.5), B=1)
   )
   &CameraOffset()
  ]
  rendererClasses [&NiagaraSpriteRendererProperties, &NiagaraLightRendererProperties]
  renderers [
   {rendererClass=&NiagaraSpriteRendererProperties}
   {rendererClass=&NiagaraLightRendererProperties, rendererIndex=1}
  ]
 }
 Laser {
  emitterUpdateScript [
   &EmitterState(
    Enable Distance Culling=true
    Enable Visibility Culling=true
    Inactive Response=::NewEnumerator1
    Life Cycle Mode=::NewEnumerator1
    Loop Behavior=::NewEnumerator1
    Loop Duration=1
    MaxDistance=10000
    Module.bUseMaxDistance=true
    Module.bUseMinDistance=false
    Scalability Mode=::NewEnumerator1
    UseLoopDelay=false
   )
   &SpawnBurst_Instantaneous(Use Spawn Probability=false, UseLoopCountLimit=true)
   &SpawnBurst_Instantaneous(Spawn Group=1, Spawn Time=0.03, Use Spawn Probability=false, UseLoopCountLimit=true) {module=SpawnBurst_Instantaneous001}
   &SpawnBurst_Instantaneous(Spawn Group=2, Spawn Time=0.07, Use Spawn Probability=false, UseLoopCountLimit=true) {module=SpawnBurst_Instantaneous002}
   &SpawnBurst_Instantaneous(Spawn Group=3, Spawn Time=0.18, Use Spawn Probability=false, UseLoopCountLimit=true) {module=SpawnBurst_Instantaneous003}
   &SpawnBurst_Instantaneous(Spawn Group=4, Spawn Time=0.2, Use Spawn Probability=false, UseLoopCountLimit=true) {module=SpawnBurst_Instantaneous004}
   &SpawnBurst_Instantaneous(Spawn Group=5, Spawn Time=0.26, Use Spawn Probability=false, UseLoopCountLimit=true) {module=SpawnBurst_Instantaneous005}
   &SpawnBurst_Instantaneous(Spawn Group=6, Spawn Time=0.2, Use Spawn Probability=false, UseLoopCountLimit=true) {module=SpawnBurst_Instantaneous006}
   &SpawnBurst_Instantaneous(Spawn Group=7, Spawn Time=0.03, Use Spawn Probability=false, UseLoopCountLimit=true) {module=SpawnBurst_Instantaneous007}
   &SpawnBurst_Instantaneous(Spawn Group=8, Spawn Time=0.03, Use Spawn Probability=false, UseLoopCountLimit=true) {module=SpawnBurst_Instantaneous008}
   &SpawnBurst_Instantaneous(Spawn Group=9, Spawn Time=0.12, Use Spawn Probability=false, UseLoopCountLimit=true) {module=SpawnBurst_Instantaneous009}
   &SpawnBurst_Instantaneous(Spawn Group=10, Spawn Time=0.2, Use Spawn Probability=false, UseLoopCountLimit=true) {module=SpawnBurst_Instantaneous010}
   &SpawnBurst_Instantaneous(Spawn Group=11, Spawn Time=0.25, Use Spawn Probability=false, UseLoopCountLimit=true) {module=SpawnBurst_Instantaneous011}
   &SpawnBurst_Instantaneous(Spawn Group=12, Spawn Time=0.22, Use Spawn Probability=false, UseLoopCountLimit=true) {module=SpawnBurst_Instantaneous012}
  ]
  particleSpawnScript [
   &InitializeParticle(
    Color=<rgba 50 0 0 1>
    Lifetime=0.25
    Mesh Scale=<0.04 5 1>
    Mesh Scale Mode=::NewEnumerator4
    Position Offset=<32 0 0>
    Sprite Size Mode=::NewEnumerator3
    Sprite UV Mode=::NewEnumerator4
    Sprite UV Scale:Vector2f=<1 -1>
    Uniform Sprite Size=3
    UsePositionOffset=true
    Write Material Random=true
   )
   &InitialMeshOrientation(
    Mesh Orientation Mode=::NewEnumerator2
    Rotation=&RandomRangeVector(Minimum=<0 -0.05 -0.1>, Maximum=<0 0.05 0.1>)
    Use Rotation Vector=true
   )
   &ShapeLocation(
    Box Midpoint=<0 0.5 0.5>
    Box Size=<45 100 65>
    MaskSpawnGroup=false
    Offset=<0 30 0>
    Offset Mode=::NewEnumerator0
    Override Seed=false
    Shape Primitive=::NewEnumerator2
    Surface Only=false
    UseOwnerScale=true
   )
   /SetVariables(Particles.PivotOffset:Vector2f=<0.5 -0.1>) {isSetParameters=true}
  ]
  particleUpdateScript [
   &PointAttractionForce(
    Attraction Radius=1000
    Attraction Strength=1300
    Attractor Position:NiagaraPosition=&MaskVectorBySpawnGroup(
     Masked Vector=<-6 0 0>
     Passthrough Vector=&MaskVectorBySpawnGroup(
      Masked Vector=&AddVectorToPosition(Position=<0 21 -2>, Vector=<-6 0 0>)
      Passthrough Vector=&MaskVectorBySpawnGroup(
       Masked Vector=&AddVectorToPosition(Position=<0 40 -3>, Vector=<-6 0 0>)
       Passthrough Vector=&MaskVectorBySpawnGroup(
        Masked Vector=&AddVectorToPosition(Position=<0 34 -10>, Vector=<-6 0 0>)
        Passthrough Vector=&MaskVectorBySpawnGroup(
         Masked Vector=&AddVectorToPosition(Position=<0 54 -7>, Vector=<-6 0 0>)
         Passthrough Vector=&MaskVectorBySpawnGroup(
          Masked Vector=&AddVectorToPosition(Position=<0 12 -8>, Vector=<-6 0 0>)
          Passthrough Vector=&MaskVectorBySpawnGroup(
           Masked Vector=&AddVectorToPosition(Position=<0 -2 -6>, Vector=<-6 0 0>)
           Passthrough Vector=&MaskVectorBySpawnGroup(
            Masked Vector=&AddVectorToPosition(Position=<0 19 -12>, Vector=<-6 0 0>)
            Passthrough Vector=&MaskVectorBySpawnGroup(
             Masked Vector=&AddVectorToPosition(Position=<0 45 -12>, Vector=<-6 0 0>)
             Passthrough Vector=&MaskVectorBySpawnGroup(
              Masked Vector=&AddVectorToPosition(Position=<0 36 -15>, Vector=<-6 0 0>)
              Passthrough Vector=&MaskVectorBySpawnGroup(
               Masked Vector=&AddVectorToPosition(Position=<0 41 -15>, Vector=<-6 0 0>)
               Passthrough Vector=&MaskVectorBySpawnGroup(
                Masked Vector=&AddVectorToPosition(Position=<0 6 -14>, Vector=<-6 0 0>)
                Passthrough Vector=&AddVectorToPosition(Position=<0 -6 -10>, Vector=<-6 0 0>)
                Spawn Group Mask=11
               )
               Spawn Group Mask=10
              )
              Spawn Group Mask=9
             )
             Spawn Group Mask=8
            )
            Spawn Group Mask=7
           )
           Spawn Group Mask=6
          )
          Spawn Group Mask=5
         )
         Spawn Group Mask=4
        )
        Spawn Group Mask=3
       )
       Spawn Group Mask=2
      )
      Spawn Group Mask=1
     )
    )
    Kill Within Radius=false
    Module.Use Falloff=true
    UseAttractorOffset=false
   )
   &Drag(
    Drag=&FloatFromCurve(
     FloatCurve=DataInterface() {propertyValues {Curve {keys [
         {interpMode=RCIM_Cubic, tangentMode=RCTM_User, time=0, value=0}
         {arriveTangent=1.953, interpMode=RCIM_Cubic, leaveTangent=1.953, tangentMode=RCTM_User, time=1, value=1}
        ]}}}
     Scale Curve=30.1338
    )
    Use Linear Drag=true
    Use Rotational Drag=true
   )
   &SolveForcesAndVelocity(Clamp Velocity=false, Limit Acceleration=false)
   &ParticleState()
   &ScaleSpriteSize(
    Non-Uniform Curve Scale=<1 4>
    Non-Uniform Curve Sprite Scale=DataInterface() {propertyValues {
      XCurve {keys [
        {arriveTangent=0.0609, interpMode=RCIM_Cubic, leaveTangent=0.0609, tangentMode=RCTM_User, time=0, value=0}
        {arriveTangent=1.1837, interpMode=RCIM_Cubic, leaveTangent=1.1837, tangentMode=RCTM_User, time=1, value=1}
       ]}
      YCurve {keys [<at 0 0 RCIM_Cubic>, {arriveTangent=2.8626, interpMode=RCIM_Cubic, leaveTangent=2.8626, tangentMode=RCTM_User, time=1, value=2}]}
     }}
    Scale Sprite Size Mode=::NewEnumerator3
   )
   &ScaleSpriteSize(
    Scale Sprite Size Mode=::NewEnumerator0
    Uniform Scale Factor=&FloatFromCurve(
     FloatCurve=DataInterface() {propertyValues {Curve {keys [
         {tangentMode=RCTM_User, time=0, value=1}
         {arriveTangent=-1.8189, interpMode=RCIM_Cubic, leaveTangent=-1.8189, tangentMode=RCTM_User, time=0.5, value=0.975}
         {arriveTangent=-2.0083, interpMode=RCIM_Cubic, leaveTangent=-2.0083, tangentMode=RCTM_User, time=1, value=0}
        ]}}}
     Scale Curve=1.3
    )
   ) {module=ScaleSpriteSize001}
   &DynamicMaterialParameters(
    Index 0 Param 1=2
    Index 0 Param 2=25
    Index 0 Param 4=1
    Module.Param0WriteEnabled=true
    Module.Param1WriteEnabled=true
    Module.Param2WriteEnabled=false
    Module.Param3WriteEnabled=false
   )
   &ScaleSpriteSizeBySpeed(Max Scale Factor=<0.75 5>, Min Scale Factor=<0.75 0.5>)
   &ScaleColor(
    Scale Alpha=&FloatFromCurve(FloatCurve=DataInterface() {propertyValues {Curve {keys [
         {arriveTangent=9.0725, interpMode=RCIM_Cubic, leaveTangent=9.0725, tangentMode=RCTM_User, time=0, value=0}
         {arriveTangent=-2.5249, interpMode=RCIM_Cubic, leaveTangent=-2.5249, tangentMode=RCTM_User, time=0.1408, value=0.9847}
         {arriveTangent=-0.0405, interpMode=RCIM_Cubic, leaveTangent=-0.0405, tangentMode=RCTM_User, time=0.8, value=0}
        ]}}})
    ScaleA=true
    ScaleRGB=false
   )
  ]
  rendererClasses [&NiagaraSpriteRendererProperties]
  renderers [{rendererClass=&NiagaraSpriteRendererProperties}]
 }
 Reload_Smoke {
  emitterSpawnScript [/SetVariables(Emitter.RandomSize:NiagaraFloat=&RandomRangeFloat(Minimum=0, Maximum=1)) {isSetParameters=true}]
  emitterUpdateScript [
   &EmitterState(
    Enable Distance Culling=true
    Enable Visibility Culling=true
    Life Cycle Mode=::NewEnumerator1
    Loop Behavior=::NewEnumerator1
    Loop Delay=&RandomRangeFloat(Minimum=0.2, Maximum=0.3)
    Loop Duration=&Lerp_Float(A=1, B=1.3, Alpha=$Emitter.RandomSize)
    MaxDistance=6000
    Module.bUseMaxDistance=true
    Module.bUseMinDistance=false
    Scalability Mode=::NewEnumerator1
    UseLoopDelay=true
   )
   &SpawnRate(
    SpawnRate=&MakeCustomFloatFromBool(
     Bool=$System.Spawn
     True Float=&FloatFromCurve(
      FloatCurve=DataInterface() {propertyValues {Curve {keys [
          {tangentMode=RCTM_User, time=0, value=1}
          {arriveTangent=-0.0405, interpMode=RCIM_Cubic, leaveTangent=-0.0405, tangentMode=RCTM_User, time=0.7009, value=1}
          {arriveTangent=-8.0089, interpMode=RCIM_Cubic, leaveTangent=-8.0089, tangentMode=RCTM_User, time=1, value=0}
         ]}}}
      CurveIndex=$Emitter.NormalizedLoopAge
      Scale Curve=60
     )
     False Float=0
    )
    Use Spawn Probability=false
   )
   &SpawnRate(
    SpawnGroup=1
    SpawnRate=&MakeCustomFloatFromBool(
     Bool=$System.Spawn
     True Float=&FloatFromCurve(
      FloatCurve=DataInterface() {propertyValues {Curve {keys [
          {tangentMode=RCTM_User, time=0, value=1}
          {arriveTangent=-0.0405, interpMode=RCIM_Cubic, leaveTangent=-0.0405, tangentMode=RCTM_User, time=0.7009, value=1}
          {arriveTangent=-8.0089, interpMode=RCIM_Cubic, leaveTangent=-8.0089, tangentMode=RCTM_User, time=1, value=0}
         ]}}}
      CurveIndex=$Emitter.NormalizedLoopAge
      Scale Curve=60
     )
     False Float=0
    )
    Use Spawn Probability=false
   ) {module=SpawnRate001}
   &SpawnRate(
    SpawnGroup=2
    SpawnRate=&MakeCustomFloatFromBool(
     Bool=$System.Spawn
     True Float=&FloatFromCurve(
      FloatCurve=DataInterface() {propertyValues {Curve {keys [
          {tangentMode=RCTM_User, time=0, value=1}
          {arriveTangent=-0.0405, interpMode=RCIM_Cubic, leaveTangent=-0.0405, tangentMode=RCTM_User, time=0.7009, value=1}
          {arriveTangent=-8.0089, interpMode=RCIM_Cubic, leaveTangent=-8.0089, tangentMode=RCTM_User, time=1, value=0}
         ]}}}
      CurveIndex=$Emitter.NormalizedLoopAge
      Scale Curve=60
     )
     False Float=0
    )
    Use Spawn Probability=false
   ) {module=SpawnRate002}
   &SpawnRate(
    SpawnGroup=3
    SpawnRate=&MakeCustomFloatFromBool(
     Bool=$System.Spawn
     True Float=&FloatFromCurve(
      FloatCurve=DataInterface() {propertyValues {Curve {keys [
          {tangentMode=RCTM_User, time=0, value=1}
          {arriveTangent=-0.0405, interpMode=RCIM_Cubic, leaveTangent=-0.0405, tangentMode=RCTM_User, time=0.7009, value=1}
          {arriveTangent=-8.0089, interpMode=RCIM_Cubic, leaveTangent=-8.0089, tangentMode=RCTM_User, time=1, value=0}
         ]}}}
      CurveIndex=$Emitter.NormalizedLoopAge
      Scale Curve=60
     )
     False Float=0
    )
    Use Spawn Probability=false
   ) {module=SpawnRate003}
   &SpawnRate(
    SpawnGroup=4
    SpawnRate=&MakeCustomFloatFromBool(
     Bool=$System.Spawn
     True Float=&FloatFromCurve(
      FloatCurve=DataInterface() {propertyValues {Curve {keys [
          {tangentMode=RCTM_User, time=0, value=1}
          {arriveTangent=-0.0405, interpMode=RCIM_Cubic, leaveTangent=-0.0405, tangentMode=RCTM_User, time=0.7009, value=1}
          {arriveTangent=-8.0089, interpMode=RCIM_Cubic, leaveTangent=-8.0089, tangentMode=RCTM_User, time=1, value=0}
         ]}}}
      CurveIndex=$Emitter.NormalizedLoopAge
      Scale Curve=60
     )
     False Float=0
    )
    Use Spawn Probability=false
   ) {module=SpawnRate004}
   &SpawnRate(
    SpawnGroup=5
    SpawnRate=&MakeCustomFloatFromBool(
     Bool=$System.Spawn
     True Float=&FloatFromCurve(
      FloatCurve=DataInterface() {propertyValues {Curve {keys [
          {tangentMode=RCTM_User, time=0, value=1}
          {arriveTangent=-0.0405, interpMode=RCIM_Cubic, leaveTangent=-0.0405, tangentMode=RCTM_User, time=0.7009, value=1}
          {arriveTangent=-8.0089, interpMode=RCIM_Cubic, leaveTangent=-8.0089, tangentMode=RCTM_User, time=1, value=0}
         ]}}}
      CurveIndex=$Emitter.NormalizedLoopAge
      Scale Curve=60
     )
     False Float=0
    )
    Use Spawn Probability=false
   ) {module=SpawnRate005}
   &SpawnRate(
    SpawnGroup=6
    SpawnRate=&MakeCustomFloatFromBool(
     Bool=$System.Spawn
     True Float=&FloatFromCurve(
      FloatCurve=DataInterface() {propertyValues {Curve {keys [
          {tangentMode=RCTM_User, time=0, value=1}
          {arriveTangent=-0.0405, interpMode=RCIM_Cubic, leaveTangent=-0.0405, tangentMode=RCTM_User, time=0.7009, value=1}
          {arriveTangent=-8.0089, interpMode=RCIM_Cubic, leaveTangent=-8.0089, tangentMode=RCTM_User, time=1, value=0}
         ]}}}
      CurveIndex=$Emitter.NormalizedLoopAge
      Scale Curve=60
     )
     False Float=0
    )
    Use Spawn Probability=false
   ) {module=SpawnRate006}
   &SpawnRate(
    SpawnGroup=7
    SpawnRate=&MakeCustomFloatFromBool(
     Bool=$System.Spawn
     True Float=&FloatFromCurve(
      FloatCurve=DataInterface() {propertyValues {Curve {keys [
          {tangentMode=RCTM_User, time=0, value=1}
          {arriveTangent=-0.0405, interpMode=RCIM_Cubic, leaveTangent=-0.0405, tangentMode=RCTM_User, time=0.7009, value=1}
          {arriveTangent=-8.0089, interpMode=RCIM_Cubic, leaveTangent=-8.0089, tangentMode=RCTM_User, time=1, value=0}
         ]}}}
      CurveIndex=$Emitter.NormalizedLoopAge
      Scale Curve=60
     )
     False Float=0
    )
    Use Spawn Probability=false
   ) {module=SpawnRate007}
   &SpawnRate(
    SpawnGroup=8
    SpawnRate=&MakeCustomFloatFromBool(
     Bool=$System.Spawn
     True Float=&FloatFromCurve(
      FloatCurve=DataInterface() {propertyValues {Curve {keys [
          {tangentMode=RCTM_User, time=0, value=1}
          {arriveTangent=-0.0405, interpMode=RCIM_Cubic, leaveTangent=-0.0405, tangentMode=RCTM_User, time=0.7009, value=1}
          {arriveTangent=-8.0089, interpMode=RCIM_Cubic, leaveTangent=-8.0089, tangentMode=RCTM_User, time=1, value=0}
         ]}}}
      CurveIndex=$Emitter.NormalizedLoopAge
      Scale Curve=60
     )
     False Float=0
    )
    Use Spawn Probability=false
   ) {module=SpawnRate008}
   &SpawnRate(
    SpawnGroup=9
    SpawnRate=&MakeCustomFloatFromBool(
     Bool=$System.Spawn
     True Float=&FloatFromCurve(
      FloatCurve=DataInterface() {propertyValues {Curve {keys [
          {tangentMode=RCTM_User, time=0, value=1}
          {arriveTangent=-0.0405, interpMode=RCIM_Cubic, leaveTangent=-0.0405, tangentMode=RCTM_User, time=0.7009, value=1}
          {arriveTangent=-8.0089, interpMode=RCIM_Cubic, leaveTangent=-8.0089, tangentMode=RCTM_User, time=1, value=0}
         ]}}}
      CurveIndex=$Emitter.NormalizedLoopAge
      Scale Curve=60
     )
     False Float=0
    )
    Use Spawn Probability=false
   ) {module=SpawnRate009}
   &SpawnRate(
    SpawnGroup=10
    SpawnRate=&MakeCustomFloatFromBool(
     Bool=$System.Spawn
     True Float=&FloatFromCurve(
      FloatCurve=DataInterface() {propertyValues {Curve {keys [
          {tangentMode=RCTM_User, time=0, value=1}
          {arriveTangent=-0.0405, interpMode=RCIM_Cubic, leaveTangent=-0.0405, tangentMode=RCTM_User, time=0.7009, value=1}
          {arriveTangent=-8.0089, interpMode=RCIM_Cubic, leaveTangent=-8.0089, tangentMode=RCTM_User, time=1, value=0}
         ]}}}
      CurveIndex=$Emitter.NormalizedLoopAge
      Scale Curve=60
     )
     False Float=0
    )
    Use Spawn Probability=false
   ) {module=SpawnRate010}
   &SpawnRate(
    SpawnGroup=11
    SpawnRate=&MakeCustomFloatFromBool(
     Bool=$System.Spawn
     True Float=&FloatFromCurve(
      FloatCurve=DataInterface() {propertyValues {Curve {keys [
          {tangentMode=RCTM_User, time=0, value=1}
          {arriveTangent=-0.0405, interpMode=RCIM_Cubic, leaveTangent=-0.0405, tangentMode=RCTM_User, time=0.7009, value=1}
          {arriveTangent=-8.0089, interpMode=RCIM_Cubic, leaveTangent=-8.0089, tangentMode=RCTM_User, time=1, value=0}
         ]}}}
      CurveIndex=$Emitter.NormalizedLoopAge
      Scale Curve=60
     )
     False Float=0
    )
    Use Spawn Probability=false
   ) {module=SpawnRate011}
   &SpawnRate(
    SpawnGroup=12
    SpawnRate=&MakeCustomFloatFromBool(
     Bool=$System.Spawn
     True Float=&FloatFromCurve(
      FloatCurve=DataInterface() {propertyValues {Curve {keys [
          {tangentMode=RCTM_User, time=0, value=1}
          {arriveTangent=-0.0405, interpMode=RCIM_Cubic, leaveTangent=-0.0405, tangentMode=RCTM_User, time=0.7009, value=1}
          {arriveTangent=-8.0089, interpMode=RCIM_Cubic, leaveTangent=-8.0089, tangentMode=RCTM_User, time=1, value=0}
         ]}}}
      CurveIndex=$Emitter.NormalizedLoopAge
      Scale Curve=60
     )
     False Float=0
    )
    Use Spawn Probability=false
   ) {module=SpawnRate012}
  ]
  particleSpawnScript [
   /SetVariables(Particles.RibbonID=/Niagara/DynamicInputs/Meshes/MakeRibbonIDFromSpawnGroup()) {isSetParameters=true}
   &InitializeParticle(
    Color Maximum=<rgba 1 1 1 0.05>
    Color Minimum=<rgba 1 1 1 0.03>
    Color Mode=::NewEnumerator2
    Lifetime=&RandomRangeFloat(
     Minimum=&FloatFromCurve(
      FloatCurve=DataInterface() {propertyValues {Curve {keys [
          <at 0 1 RCIM_Cubic>
          {arriveTangent=-1.883, interpMode=RCIM_Cubic, leaveTangent=-1.883, tangentMode=RCTM_User, time=0.7, value=0.5}
          <at 0.8 0 RCIM_Cubic>
         ]}}}
      CurveIndex=$Emitter.NormalizedLoopAge
      Scale Curve=&Lerp_Float(A=0.5, B=0.7, Alpha=$Emitter.RandomSize)
     )
     Maximum=&FloatFromCurve(
      FloatCurve=DataInterface() {propertyValues {Curve {keys [
          {arriveTangent=0.625, interpMode=RCIM_Cubic, time=0, value=0.9}
          {arriveTangent=-1.883, interpMode=RCIM_Cubic, leaveTangent=-1.883, tangentMode=RCTM_User, time=0.9, value=0.5}
          <at 1 0 RCIM_Cubic>
         ]}}}
      CurveIndex=$Emitter.NormalizedLoopAge
      Scale Curve=&Lerp_Float(A=0.5, B=0.7, Alpha=$Emitter.RandomSize)
     )
    )
    Position Offset=<-6 0 -3>
    Ribbon Width=&Lerp_Float(A=4, B=11, Alpha=$Emitter.RandomSize)
    Ribbon Width Mode=::NewEnumerator1
    UsePositionOffset=true
    Write Material Random=true
   )
   /SetVariables(Particles.Position=&MaskVectorBySpawnGroup(
     Masked Vector=<-6 0 -3>
     Passthrough Vector=&MaskVectorBySpawnGroup(
      Masked Vector=&AddVectorToPosition(Position=<0 21 -2>, Vector=<-6 0 -4>)
      Passthrough Vector=&MaskVectorBySpawnGroup(
       Masked Vector=&AddVectorToPosition(Position=<0 40 -3>, Vector=<-6 0 -3>)
       Passthrough Vector=&MaskVectorBySpawnGroup(
        Masked Vector=&AddVectorToPosition(Position=<0 34 -10>, Vector=<-6 0 -3>)
        Passthrough Vector=&MaskVectorBySpawnGroup(
         Masked Vector=&AddVectorToPosition(Position=<0 54 -7>, Vector=<-6 0 -3>)
         Passthrough Vector=&MaskVectorBySpawnGroup(
          Masked Vector=&AddVectorToPosition(Position=<0 12 -8>, Vector=<-6 0 -3>)
          Passthrough Vector=&MaskVectorBySpawnGroup(
           Masked Vector=&AddVectorToPosition(Position=<0 -2 -6>, Vector=<-6 0 -3>)
           Passthrough Vector=&MaskVectorBySpawnGroup(
            Masked Vector=&AddVectorToPosition(Position=<0 19 -12>, Vector=<-6 0 -3>)
            Passthrough Vector=&MaskVectorBySpawnGroup(
             Masked Vector=&AddVectorToPosition(Position=<0 45 -12>, Vector=<-6 0 -3>)
             Passthrough Vector=&MaskVectorBySpawnGroup(
              Masked Vector=&AddVectorToPosition(Position=<0 36 -15>, Vector=<-6 0 -3>)
              Passthrough Vector=&MaskVectorBySpawnGroup(
               Masked Vector=&AddVectorToPosition(Position=<0 41 -15>, Vector=<-6 0 -3>)
               Passthrough Vector=&MaskVectorBySpawnGroup(
                Masked Vector=&AddVectorToPosition(Position=<0 6 -14>, Vector=<-6 0 -3>)
                Passthrough Vector=&AddVectorToPosition(Position=<0 -6 -10>, Vector=<-6 0 -3>)
                Spawn Group Mask=11
               )
               Spawn Group Mask=10
              )
              Spawn Group Mask=9
             )
             Spawn Group Mask=8
            )
            Spawn Group Mask=7
           )
           Spawn Group Mask=6
          )
          Spawn Group Mask=5
         )
         Spawn Group Mask=4
        )
        Spawn Group Mask=3
       )
       Spawn Group Mask=2
      )
      Spawn Group Mask=1
     )
    )) {isSetParameters=true}
   &AddVelocity(
    Coordinate Space=::World
    Velocity=&MakeVector(
     X=0
     Y=0
     Z=&Multiply_Float(
      A=40
      B=&FloatFromCurve(FloatCurve=DataInterface() {propertyValues {Curve {keys [<at 0.5 1>,<at 1 0>]}}}, CurveIndex=$Emitter.NormalizedLoopAge)
     )
    )
   )
  ]
  particleUpdateScript [
   &ParticleState()
   &ScaleColor(
    Scale Alpha=&Multiply_Float(
     A=&FloatFromCurve(
      FloatCurve=DataInterface() {propertyValues {Curve {keys [<at 0 0>, {arriveTangent=-1.7714, interpMode=RCIM_Cubic, leaveTangent=-1.7714, tangentMode=RCTM_User, time=0.1, value=1}, <at 0.8 0>]}}}
      Scale Curve=&FNTimeOfDayScalar(Night Scale=1, Morning Scale=0.3, Evening Scale=0.5)
     )
     B=&FloatFromCurve(
      FloatCurve=DataInterface() {propertyValues {Curve {keys [<at 0 1>, {arriveTangent=-1.4286, interpMode=RCIM_Cubic, leaveTangent=-1.4286, time=0.1411, value=1}, <at 0.7 0 RCIM_Cubic>]}}}
      CurveIndex=$Emitter.NormalizedLoopAge
      Scale Curve=0.4
     )
    )
    Scale RGB=&MakeVectorFromLinearColorRGB(Color=<rgba 1 1 1 0.1>)
    ScaleA=true
    ScaleRGB=true
   )
   &DynamicMaterialParameters(
    Index 0 Param 1=&FloatFromCurve(FloatCurve=DataInterface() {propertyValues {Curve {keys [
         <at 0 1 RCIM_Cubic>
         {arriveTangent=-1.7387, arriveTangentWeight=1.9926, interpMode=RCIM_Cubic, leaveTangent=-1.7387, tangentMode=RCTM_User, time=1, value=0.125}
        ]}}})
    Index 0 Param 3=-1
    Index 0 Param 4=&FloatFromCurve(FloatCurve=DataInterface() {propertyValues {Curve {keys [{arriveTangent=-2.94, interpMode=RCIM_Cubic, time=0, value=1}, <at 0.75 0 RCIM_Cubic>]}}})
    Index 1 Param 1=&FloatFromCurve(FloatCurve=DataInterface() {propertyValues {Curve {keys [<at 0 1 RCIM_Cubic>, {arriveTangent=45.3807, interpMode=RCIM_Cubic, leaveTangent=45.3807, tangentMode=RCTM_User, time=0.5, value=10}]}}})
    Index 1 Param 2=&FloatFromCurve(FloatCurve=DataInterface() {propertyValues {Curve {keys [{arriveTangent=-2.2405, interpMode=RCIM_Cubic, leaveTangent=-2.2405, tangentMode=RCTM_User, time=0, value=1}, <at 1 0 RCIM_Cubic>]}}})
    Module.Param0WriteEnabled=true
    Module.Param0WriteEnabled1=true
    Module.Param1WriteEnabled=false
    Module.Param1WriteEnabled1=true
    Module.Param2WriteEnabled=true
    Module.Param2WriteEnabled1=false
    Module.Param3WriteEnabled=true
    Module.Param3WriteEnabled1=false
    Write Parameter Index 1=true
   )
   &ScaleRibbonWidth(Ribbon Width Scale=&FloatFromCurve(FloatCurve=DataInterface() {propertyValues {Curve {keys [
         {arriveTangent=9.0725, interpMode=RCIM_Cubic, leaveTangent=9.0725, tangentMode=RCTM_User, time=0, value=0}
         {arriveTangent=-2.5249, interpMode=RCIM_Cubic, leaveTangent=-2.5249, tangentMode=RCTM_User, time=0.1408, value=0.9847}
         {arriveTangent=-0.0405, interpMode=RCIM_Cubic, leaveTangent=-0.0405, tangentMode=RCTM_User, time=1, value=0}
        ]}}}))
   &CurlNoiseForce(
    Mask Curl Noise=true
    Noise Quality / Cost=::NewEnumerator0
    Noise Strength=&RandomRangeFloat(Minimum=500, Maximum=510)
    Pan Noise=true
    Pan Noise Field=<5 0 0>
   )
   &SolveForcesAndVelocity(Clamp Velocity=false, Limit Acceleration=false)
   &ScaleColor(
    Scale Alpha=&LerpMultipleFloats(
     Float Count:ENiagara_MultipleLerpCount=::NewEnumerator1
     1:NiagaraFloat=1
     2:NiagaraFloat=1.5
     3:NiagaraFloat=0.75
     4:NiagaraFloat=0.4
     5:NiagaraFloat=?
     Alpha:NiagaraFloat=Linked() {linkedVariable {name=NPC.FortniteNPC.TimeOfDay_0-1, type=NiagaraFloat}}
    )
    ScaleA=true
    ScaleRGB=false
   ) {module=ScaleColor002}
  ]
  rendererClasses [&NiagaraRibbonRendererProperties]
  renderers [{rendererClass=&NiagaraRibbonRendererProperties}]
 }
}
summary {
 systemName=NS_Emote_ChillCat_Laser_Sprite
 emitters [
  {emitterName=Laser, rendererClasses [&NiagaraSpriteRendererProperties]}
  {emitterName=Dust, simTarget=GPUComputeSim, rendererClasses [&NiagaraSpriteRendererProperties]}
  {emitterName=Glow, rendererClasses [&NiagaraSpriteRendererProperties, &NiagaraLightRendererProperties]}
  {emitterName=Reload_Smoke, rendererClasses [&NiagaraRibbonRendererProperties]}
 ]
 userVariables [
  User.Alpha {type=NiagaraFloat, defaultValue=1, description=""}
  User.bIsFrontEnd {type=NiagaraBool}
  User.bIsFrontEndPreview {type=NiagaraBool}
 ]
}
